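# Looping player foley cue for sliding on ice, using the small IceSurface slide loop wave.
# A continuous modulator maps the Speed parameter to volume and a narrow pitch range, and a second scales volume by OnGround.
# An enveloper at the root applies a very short fade-in to avoid an abrupt start.
# Player-specific attenuation applies, and concurrency is overridden to limit voices per owner, stopping the quietest.
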SoundCue /FNGameplayCues/Sounds/Fort_Traps/Ice_Trap/Environment/IceLake_SlideLoop_Player_Cue
root=Enveloper_1
nodes [
 {
  id=Enveloper_1
  children [ModulatorContinuous_1]
  properties {
   loopEnd=0.25
   loopStart=0.25
   pitchCurve {editorCurveData {keys [{time=0}]}}
   volumeCurve {editorCurveData {keys [<at 0 -0.0003 RCIM_Cubic>,<at 0.0156 1.0006>]}}
  }
 }
 {
  id=ModulatorContinuous_0
  children [WavePlayer_7]
  properties {
   pitchModulationParams {default=0, maxOutput=1.1, minOutput=0.9, parameterName=Speed}
   volumeModulationParams {default=0, parameterName=Speed}
  }
 }
 {id=ModulatorContinuous_1, children [ModulatorContinuous_0], properties {volumeModulationParams {parameterName=OnGround}}}
 {
  id=WavePlayer_7
  properties {bLooping=true, soundWaveAssetPtr=&/IceSurface_SlideLoop_Small_A_01}
 }
]
properties {
 attenuationSettings=&/IceSurface_SlideLoop_Player_Att
 bOverrideConcurrency=true
 maxDistance=4084
 soundClassObject=&Fort_Player_Foley
 volumeMultiplier=2.5
 concurrencyOverrides {bLimitToOwner=true, maxCount=4, resolutionRule=StopQuietest}
}
