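# Looping stereo MetaSound for a player sliding on ice, layering an ice-surface loop with a wind loop.
# Speed and its absolute difference from a smoothed copy drive ice gain, pitch and waveshaper amount.
# OnGround scales the ice layer, while footsteps briefly duck it and randomize its pitch.
# Waveshaped wind rises at higher speeds for the local player only, and an attack envelope fades in the whole mix on play.
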MetaSoundSource /FNGameplayCues/Sounds/Fort_Traps/Ice_Trap/Environment/IceLake_SlideLoop_Player_MSS
inputs [
 IceLoop {type=WaveAsset}
 IsLocal {type=Bool}
 OnFootstep {type=Trigger}
 OnGround {type=Float}
 Speed {type=Float}
 UE.Source.OnPlay {type=Trigger}
 WindLoop {type=WaveAsset}
]
interfaces [UE.OutputFormat.Stereo:1.0, UE.Source:1.1]
outputs [
 "UE.OutputFormat.Stereo.Audio:0" {type=Audio}
 "UE.OutputFormat.Stereo.Audio:1" {type=Audio}
]
pages [Default {
  input_defaults {
   IceLoop=&/IceSurface_SlideLoop_Small_A_01
   IsLocal=null
   OnFootstep=null
   OnGround=1
   Speed=0.4
   UE.Source.OnPlay=null
   WindLoop=/FNGameplayCues/Athena/Sounds/Wind_Falling/Character_InAir_Wind_Lp_01
  }
  nodes [
   ADSREnvelopeADSREnvelope_0 = ADSR Envelope.ADSR Envelope:Float(Trigger Attack=^UE.Source.OnPlay) {also {defaults {Attack Curve=0.5, Attack Time=1, Release Time=0.2, Sustain Level=1}}}
   n1 = UE.Subtract:Float(PrimaryOperand=^Speed, AdditionalOperands=UE.InterpTo:Audio(Target=^Speed).Value)
   n2 = VariableAccessor.Float(Variable=VariableMutator.Float(Value=AbsoluteValue.Abs:Float(Input=@n1.Out).Value, Variable=InitVariable.Float().Variable).Variable)
   n3 = VariableAccessor.Float(Variable=@n2.Variable)
   n4 = Max.Max:Float(A=^Speed, B=VariableAccessor.Float(Variable=@n3.Variable).Value)
   ADSREnvelopeADSREnvelope_1 = ADSR Envelope.ADSR Envelope:Float(Trigger Attack=^OnFootstep) {also {defaults {Attack Time=0.1, Decay Time=0, Release Time=2, Sustain Level=0}}}
   n5 = VariableMutator.Float(
    Variable=InitVariable.Float().Variable
    Value=UE.Subtract:Float(AdditionalOperands=UE.Multiply:Float(PrimaryOperand=@ADSREnvelopeADSREnvelope_1.Out Envelope) {also {defaults {AdditionalOperands=0.5}}}.Out) {also {defaults {PrimaryOperand=1}}}.Out
   )
   n6 = MapRange.MapRange:Float(In=@n3.Value) {also {defaults {Out Range B=4}}}
   GetWaveDuration_0 = "UE.Get Wave Duration: "(Wave=^IceLoop)
   FrequencyMultipliertoSemitone_1 = UE.Frequency Multiplier to Semitone:Float(Frequency Multiplier=MapRange.MapRange:Float(In=^Speed) {also {defaults {Out Range A=0.9}}}.Out Value)
   n7 = UE.Add:Float(
    PrimaryOperand=UE.Add:Float(
     PrimaryOperand=@FrequencyMultipliertoSemitone_1.Semitones
     AdditionalOperands=UE.RandomFloat(Next=^OnFootstep) {also {defaults {Min=-1}}}.Value
    ).Out
    AdditionalOperands=MapRange.MapRange:Float(In=@n2.Value) {also {defaults {Out Range B=2}}}.Out Value
   )
   n8 = InitVariable.Float()
   WavePlayer_0 = UE.Wave Player:Mono(
    Wave Asset=^IceLoop
    Play=^UE.Source.OnPlay
    Start Time=UE.RandomTime(Next=^UE.Source.OnPlay, Max=@GetWaveDuration_0.Duration).Value
    Pitch Shift=VariableAccessor.Float(Variable=VariableMutator.Float(Value=@n7.Out, Variable=@n8.Variable).Variable).Value
   ) {also {defaults {Loop="<MetaSoundBuilderResult.SUCCEEDED: 0>"}}}
   n9 = UE.Multiply:Audio by Float(
    AdditionalOperands=UE.InterpTo:Audio(Target=^OnGround).Value
    PrimaryOperand=UE.Multiply:Audio by Float(
     AdditionalOperands=VariableAccessor.Float(Variable=@n5.Variable).Value
     PrimaryOperand=UE.WaveShaper:Audio(Amount=@n6.Out Value, In=@WavePlayer_0.Out Mono).Out
    ).Out
   )
   n10 = MapRange.MapRange:Float(In=^Speed) {also {defaults {In Range A=0.5}}}
   FrequencyMultipliertoSemitone_0 = UE.Frequency Multiplier to Semitone:Float(Frequency Multiplier=MapRange.MapRange:Float(In=^Speed) {also {defaults {Out Range A=0.9, Out Range B=1.5}}}.Out Value)
   WavePlayer_1 = UE.Wave Player:Stereo(Play=^UE.Source.OnPlay, Wave Asset=^WindLoop, Pitch Shift=@FrequencyMultipliertoSemitone_0.Semitones) {also {defaults {Loop="<MetaSoundBuilderResult.SUCCEEDED: 0>"}}}
   AudioMixerAudioMixerStereo2_0 = "AudioMixer.Audio Mixer (Stereo, 2)"(
    Gain 0=UE.InterpTo:Audio(Target=MapRange.MapRange:Float(In=@n4.Value) {also {defaults {In Range B=0.5}}}.Out Value) {also {defaults {Interp Time=2}}}.Value
    In 0 L=@n9.Out
    In 0 R=@n9.Out
    Gain 1=UE.Multiply:Float(PrimaryOperand=Convert.Float:Bool(Bool=^IsLocal).Float, AdditionalOperands=UE.InterpTo:Audio(Target=@n10.Out Value).Value).Out
    In 1 R=UE.WaveShaper:Audio(In=@WavePlayer_1.Out Right) {also {defaults {Amount=2, Type=0}}}.Out
    In 1 L=UE.WaveShaper:Audio(In=@WavePlayer_1.Out Left) {also {defaults {Amount=2, Type=0}}}.Out
   )
   Multiply_0 = UE.Multiply:Audio by Float(AdditionalOperands=@ADSREnvelopeADSREnvelope_0.Out Envelope, PrimaryOperand=@AudioMixerAudioMixerStereo2_0.Out L)
   Multiply_3 = UE.Multiply:Audio by Float(AdditionalOperands=@ADSREnvelopeADSREnvelope_0.Out Envelope, PrimaryOperand=@AudioMixerAudioMixerStereo2_0.Out R)
  ]
  links [{dst="@UE.OutputFormat.Stereo.Audio:0", src=Multiply_0.Out}, {dst="@UE.OutputFormat.Stereo.Audio:1", src=Multiply_3.Out}]
 }]
properties {
 assetImportData=&self:AssetImportData
 attenuationSettings=&/IceSurface_SlideLoop_Player_Att
 bLooping=true
 bOverrideConcurrency=true
 duration=10000
 maxDistance=4084
 numChannels=2
 outputFormat=Stereo
 soundClassObject=&Fort_Player_Foley
 concurrencyOverrides {bLimitToOwner=true, maxCount=4, resolutionRule=StopQuietest}
}
