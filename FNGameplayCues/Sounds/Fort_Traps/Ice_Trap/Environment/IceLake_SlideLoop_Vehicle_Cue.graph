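# Looping Sound Cue for a vehicle sliding on an ice surface, named for the Ice Lake environment under the Ice Trap sounds.
# A Random node picks among three equally weighted, looping large ice-slide waves.
# The pick passes through a default Modulator into an Enveloper that ramps volume up from silence.
# The cue plays slightly pitched up on the exterior vehicle sound class, with golf-cart movement attenuation and owner-limited concurrency.
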SoundCue /FNGameplayCues/Sounds/Fort_Traps/Ice_Trap/Environment/IceLake_SlideLoop_Vehicle_Cue
root=Enveloper_0
nodes [
 {
  id=Enveloper_0
  children [Modulator_0]
  properties {pitchCurve {editorCurveData {keys [{time=0}]}}, volumeCurve {editorCurveData {keys [<at 0 0>,<at 1 1>]}}}
 }
 {id=Modulator_0, children [Random_0]}
 {id=Random_0, children [WavePlayer_2,WavePlayer_3,WavePlayer_4], properties {bShouldExcludeFromBranchCulling=true, weights [1,1,1]}}
 {
  id=WavePlayer_2
  properties {bLooping=true, soundWaveAssetPtr=&/IceSurface_SlideLoop_Large_A_01}
 }
 {
  id=WavePlayer_3
  properties {bLooping=true, soundWaveAssetPtr=&/IceSurface_SlideLoop_Large_A_02}
 }
 {
  id=WavePlayer_4
  properties {bLooping=true, soundWaveAssetPtr=&/IceSurface_SlideLoop_Large_A_03}
 }
]
properties {
 attenuationSettings=/Game/Sounds/Attenuation/Fort_GolfCart_Movement_Close_Att
 maxDistance=5300
 pitchMultiplier=1.2
 soundClassObject=&Fort_Vehicles_Exterior
 volumeMultiplier=0.95
 concurrencyOverrides {bLimitToOwner=true}
}
